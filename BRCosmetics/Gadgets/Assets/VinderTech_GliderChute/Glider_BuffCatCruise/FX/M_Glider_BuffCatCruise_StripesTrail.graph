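# Unlit, two-sided translucent Niagara ribbon material for the striped trail effect of the BuffCatCruise glider.
# A sine of ribbon V, sharpened by clamping, blends two colors into stripes, which a tiled, pannable smoke noise and particle color multiply into emissive.
# Opacity combines particle alpha, a weapon-trail mask, camera depth fade and optional leading and trailing edge fades.
# A shading path switch applies DepthFade by default and skips it on mobile.
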Material /BRCosmetics/Gadgets/Assets/VinderTech_GliderChute/Glider_BuffCatCruise/FX/M_Glider_BuffCatCruise_StripesTrail
graph {
 properties {
  bUsedWithNiagaraRibbons=true
  blendMode=BLEND_Translucent
  pixelDepthOffsetMode=PDOM_Legacy
  shadingModel=MSM_Unlit
  twoSided=true
  usedShadingModels=Unlit
 }
 nodes [
  n1 = TextureCoordinate() {coordinateIndex=1}
  n2 = AppendVector(A=ScalarParameter() {defaultValue=0.1, parameterName=CloudTileU}, B=ScalarParameter() {defaultValue=1, parameterName=CloudTileV})
  n3 = ScalarParameter() {parameterName=TilingNoisePanSpeedU}
  n4 = ScalarParameter() {parameterName=TilingNoisePanSpeedV}
  n5 = TextureSampleParameter2D(UVs=Panner(Coordinate=Multiply(A=@n1, B=@n2), Speed=AppendVector(A=@n3, B=@n4)) {desc=Default Speed is 0}) {parameterName=TilingNoise, texture=/Game/Global/Textures/Default/Tileable_Patterns/T_Smoke_Puff_Tile_Up}
  n6 = ComponentMask(None=TextureCoordinate()) {g=true}
  n7 = ScalarParameter() {defaultValue=5, parameterName=Density, sliderMax=10}
  n8 = Sine(None=Add(A=Multiply(A=@n6, B=@n7) {constB=5.2}, B=ScalarParameter() {parameterName=Offset, sliderMax=1, sliderMin=-1}))
  n9 = Multiply(A=Clamp(None=Multiply(A=@n8)) {minDefault=-1}) {constB=2}
  n10 = LinearInterpolate(
   A=VectorParameter() {parameterName=Color 1, defaultValue=<rgba 1 1 1 1>}.RGB
   B=VectorParameter() {parameterName=Color 2, defaultValue=<rgba 0.0469 0.5836 1 1>}.RGB
   Alpha=Saturate(None=Add(A=@n9) {constB=0.5})
  )
  n11 = ParticleColor()
  MaterialFunctionCall(Emissive=Multiply(A=Multiply(A=@n5.G, B=@n10), B=@n11.RGB)) {materialFunction=&MF_FX_EmissiveAdjustments}
  n12 = Multiply(A=TextureCoordinate(), B=Constant2Vector() {g=0.8, r=1})
  n13 = Constant2Vector() {r=0.05}
  n14 = TextureSampleParameter2D(UVs=Saturate(None=Add(A=@n12, B=@n13))) {
   parameterName=SecondaryMask
   samplerSource=SSM_Wrap_WorldGroupSettings
   texture=/Game/Effects/Fort_Effects/Textures/Melee/T_GenericWeaponTrail
  }
  n15 = Multiply(
   A=Multiply(A=@n11.A, B=@n14.G)
   B=MaterialFunctionCall(Fade Length=ScalarParameter() {defaultValue=150, parameterName=CamDepthFadeLength}) {materialFunction=&CameraDepthFade}.Result
  )
  n16 = ComponentMask(None=TextureCoordinate()) {r=true}
  n17 = Add(A=@n16, B=ScalarParameter() {group="Fade Leading && Trailing Edges", parameterName=Leading Edge Fade Bias, sortPriority=2})
  n18 = ScalarParameter() {defaultValue=3, group="Fade Leading && Trailing Edges", parameterName=Leading Edge Fade Scale}
  n19 = LinearInterpolate(
   B=Saturate(None=Multiply(A=@n17, B=@n18) {constB=3})
   Alpha=ScalarParameter() {defaultValue=1, group=Alpha, parameterName=Fade Leading Edge, sortPriority=4}
  ) {constA=1}
  n20 = Add(A=OneMinus(None=@n16), B=ScalarParameter() {group="Fade Leading && Trailing Edges", parameterName=Trailing Edge Fade Bias, sortPriority=3})
  n21 = ScalarParameter() {defaultValue=2, group="Fade Leading && Trailing Edges", parameterName=Trailing Edge Fade Scale, sortPriority=1}
  n22 = LinearInterpolate(
   B=Saturate(None=Multiply(A=@n20, B=@n21) {constB=2})
   Alpha=ScalarParameter() {defaultValue=1, group=Alpha, parameterName=Fade Trailing Edge, sortPriority=5}
  ) {constA=1}
  n23 = LinearInterpolate(A=@n15, B=Multiply(A=@n15, B=Multiply(A=@n19, B=@n22)), Alpha=ScalarParameter() {parameterName=UseEdgeFade})
  ShadingPathSwitch(Default=DepthFade(Opacity=@n23, FadeDistance=ScalarParameter() {parameterName=DepthFadeDistance}), Mobile=@n23) {desc=Migrated from FeatureLevelSwitch}
 ]
}
